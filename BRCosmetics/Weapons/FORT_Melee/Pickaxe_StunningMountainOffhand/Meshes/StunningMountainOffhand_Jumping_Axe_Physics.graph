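# Physics asset for the StunningMountainOffhand jumping-axe pickaxe mesh, covering its front-left and back-left bolts.
# Each bolt pairs a tiny capsule on a base bone with a longer, thin capsule on a dyn bone.
# A constraint joins each dyn body to its base, with swing limits of a few degrees and a zero twist limit.
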PhysicsAsset /BRCosmetics/Weapons/FORT_Melee/Pickaxe_StunningMountainOffhand/Meshes/StunningMountainOffhand_Jumping_Axe_Physics
bodies {
 base_bolt_bk_l [{shapeType=Capsule, length=0.505, radius=0.505, shapeName=base_bolt_bk_l_capsule, rotation=<rot 0 -90 0>}]
 base_bolt_fr_l [{shapeType=Capsule, length=0.505, radius=0.505, shapeName=base_bolt_fr_l_capsule, rotation=<rot 0 -90 0>}]
 dyn_bolt_bk_l [{shapeType=Capsule, length=6, radius=1, shapeName=dyn_bolt_bk_l_capsule, rotation=<rot 90 0 0>}]
 dyn_bolt_fr_l [{shapeType=Capsule, length=6, radius=1, shapeName=dyn_bolt_fr_l_capsule, rotation=<rot 90 0 0>}]
}
constraints [
 {bone1Name=dyn_bolt_bk_l, bone2Name=base_bolt_bk_l, swing1LimitDegrees=3, swing2LimitDegrees=2, twistLimitDegrees=0}
 {bone1Name=dyn_bolt_fr_l, bone2Name=base_bolt_fr_l, swing1LimitDegrees=3, swing2LimitDegrees=2, twistLimitDegrees=0}
]
